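# Weapon animation blueprint for the Grappler Bow, built on the Explosive Bow skeleton and animations.
# A Default/Charging state machine handles aim and draw, with gendered blend spaces, curve-driven wing bone rotation, a max-charge loop and crouch additives.
# A second machine animates the grappler's charged state from arrow_mid, weighted on only when the owner casts to the Grappler Bow.
# An out-of-ammo pose layers onto arrow_nock before the reload slot.
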AnimBlueprint /PrimalGameplay/Items/Exotics/GrapplerBow/Assets/Animation/GrapplerBow_AnimBP
target_skeleton=/Game/Weapons/FORT_Crossbows/Explosive_Bow/Meshes/Explosive_Bow_Skeleton
graphs [
 AnimGraph {nodes [
   AnimGraphNode_StateMachine_0 = ExplosiveBow() {is=title AnimGraphNode_StateMachine}
   AnimGraphNode_StateMachine_1 = GrapplerBowCharge() {is=title AnimGraphNode_StateMachine}
   n1 = Layered blend per bone(
    BasePose=Slot 'DefaultSlot'(Source=@AnimGraphNode_StateMachine_0)
    BlendPoses_0=@AnimGraphNode_StateMachine_1
    BlendWeights_0=Get GrapplerBow() -> (GrapplerBow:double)
   ) {layerSetup [{branchFilters [{boneName=arrow_mid}]}]}
   n2 = Sequence Player 'ExplosiveBow_NoAmmo_Pose_Weapon'() {sequence=/Game/Animation/Game/MainPlayer/Combat/Guns/ExplosiveBow/ExplosiveBow_NoAmmo_Pose_Weapon}
   n3 = Get OutOfAmmoAlpha() -> (OutOfAmmoAlpha:float)
   Output Pose(Result=Slot 'ReloadSlot'(Source=Layered blend per bone(BasePose=@n1, BlendPoses_0=@n2, BlendWeights_0=@n3) {layerSetup [{branchFilters [{boneName=arrow_nock}]}]}) {slotName=ReloadSlot}) {layerGroup=None}
  ]}
 ExplosiveBow {
  entry=Default
  path=AnimGraph.AnimGraphNode_StateMachine_0.ExplosiveBow
  ids=id
  nodes [
   AnimStateNode_0 = Default() {is=title AnimStateNode}
   AnimStateNode_1 = Charging() {is=title AnimStateNode}
   AnimStateTransitionNode_0 = AnimStateTransitionNode() {crossfadeDuration=0.55}
   AnimStateTransitionNode_1 = AnimStateTransitionNode() {crossfadeDuration=0.1}
   AnimStateTransitionNode_4 = AnimStateTransitionNode() {crossfadeDuration=0}
  ]
  links [
   {
    dst=Default
    rule=AnimGraph.AnimGraphNode_StateMachine_0.ExplosiveBow.AnimStateTransitionNode_4.Transition
    src=Charging
    via=AnimStateTransitionNode_4
   }
   {
    dst=Charging
    rule=AnimGraph.AnimGraphNode_StateMachine_0.ExplosiveBow.AnimStateTransitionNode_0.Transition
    src=Default
    via=AnimStateTransitionNode_0
   }
   {
    dst=Charging
    rule=AnimGraph.AnimGraphNode_StateMachine_0.ExplosiveBow.AnimStateTransitionNode_1.Transition
    src=Default
    via=AnimStateTransitionNode_1
   }
  ]
 }
 Default {
  path=AnimGraph.AnimGraphNode_StateMachine_0.ExplosiveBow.AnimStateNode_0.Default
  nodes [
   n1 = Local Space Ref Pose()
   n2 = Sequence Player 'ExplosiveBow_Base_Pose_Weapon'() {sequence=/Game/Animation/Game/MainPlayer/Combat/Guns/ExplosiveBow/ExplosiveBow_Base_Pose_Weapon}
   n3 = Get bShouldPlayDefaultPose() -> (bShouldPlayDefaultPose:bool)
   n4 = Get BowAimYaw() -> (BowAimYaw:float)
   n5 = Get BowAimPitch() -> (BowAimPitch:float)
   Output Animation Pose(Result=Apply Additive(
     Base=Blend Poses by bool(BlendPose_0=@n1, BlendPose_1=@n2, bActiveValue=@n3) {blendType=Linear}
     Additive=Blendspace Player 'ExplosiveBow_Targeted_AO_BS_Weapon'(X=@n4, Y=@n5) {blendSpace=/Game/Animation/Game/MainPlayer/Poses/Medium/Male/AO/ExplosiveBow/Weapon/ExplosiveBow_Targeted_AO_BS_Weapon}
    ) {alphaBoolBlend {blendOption=Linear}})
  ]
 }
 Charging {
  path=AnimGraph.AnimGraphNode_StateMachine_0.ExplosiveBow.AnimStateNode_1.Charging
  nodes [
   n1 = Blendspace Player 'ExplosiveBow_Targeted_AO_BS_Weapon'(X=Get BowAimYaw(), Y=0) {blendSpace=&ExplosiveBow_Targeted_AO_BS_Weapon}
   n2 = Get BowAimPitch()
   n3 = Get ChargePlayRate() -> (ChargePlayRate:float)
   n4 = Get ChargeBlendSpaceCrouchSpeed() -> (ChargeBlendSpaceCrouchSpeed:float)
   n5 = "Blend Poses (EFortCustomGender)"!(
    BlendPose_0:Pose Link Structure=Blendspace Player 'ExplosiveBow_Combined_Charging_BS_Weapon'!(X=@n4, Y=@n2, PlayRate:float=@n3) -> (Pose) {bLoop=false, blendSpace=/Game/Animation/Game/MainPlayer/Combat/Guns/ExplosiveBow/ExplosiveBow_Combined_Charging_BS_Weapon}
    BlendPose_1:Pose Link Structure=Blendspace Player 'ExplosiveBow_Combined_Charging_BS_Weapon_CMF'!(X=@n4, Y=@n2, PlayRate:float=@n3) -> (Pose) {bLoop=false, blendSpace=/Game/Animation/Game/MainPlayer/Combat/Guns/ExplosiveBow/ExplosiveBow_Combined_Charging_BS_Weapon_CMF}
    BlendTime_0:float=0.1
    BlendTime_1:float=0.1
    ActiveEnumValue:Byte=Get OwnerGender() -> (OwnerGender:EFortCustomGender Enum=Invalid)
   ) -> (Pose:Pose Link Structure) {
    is=title AnimGraphNode_BlendListByEnum
    blendType=Linear
    blendPose [{linkId=-1, sourceLinkId=-1}, {linkId=-1, sourceLinkId=-1}]
    blendTime [0.1,0.1]
   }
   n6 = GetCurveValue(CurveName=Wing_Charging)
   n7 = "Transform (Modify) Bone - Bone: bow_wing_top"(ComponentPose=Local To Component(LocalPose=@n5), Alpha=1, Translation="0,0,0", Rotation=MakeRotator(Roll=@n6), Scale="1,1,1") {
    rotationMode=BMM_Additive
    rotationSpace=BCS_BoneSpace
    scaleMode=BMM_Additive
    translationMode=BMM_Additive
    alphaBoolBlend {blendOption=Linear}
    boneToModify {boneName=bow_wing_top}
   }
   n8 = "Transform (Modify) Bone - Bone: bow_wing_bott"(ComponentPose=@n7, Alpha=1, Translation="0,0,0", Rotation=MakeRotator(Roll=float * float(A=@n6, B=-1)), Scale="1,1,1") {
    rotationMode=BMM_Additive
    rotationSpace=BCS_BoneSpace
    scaleMode=BMM_Additive
    translationMode=BMM_Additive
    alphaBoolBlend {blendOption=Linear}
    boneToModify {boneName=bow_wing_bott}
   }
   n9 = "Blend Poses (EFortCustomGender)"(
    BlendPose_0=Sequence Player 'ExplosiveBow_ChargingLoop_CMM_Weapon'() {sequence=/Game/Animation/Game/MainPlayer/Combat/Guns/ExplosiveBow/ExplosiveBow_ChargingLoop_CMM_Weapon}
    BlendPose_1=Sequence Player 'ExplosiveBow_ChargingLoop_CMF_Weapon'() {sequence=/Game/Animation/Game/MainPlayer/Combat/Guns/ExplosiveBow/ExplosiveBow_ChargingLoop_CMF_Weapon}
    BlendTime_0=0.1
    BlendTime_1=0.1
    ActiveEnumValue=Get OwnerGender() -> (OwnerGender=Invalid)
   ) {blendType=Linear, blendPose [{linkId=-1, sourceLinkId=-1}, {linkId=-1, sourceLinkId=-1}], blendTime [0.1,0.1]}
   n10 = AimOffset Player 'ExplosiveBow_ChargingLoop_Weapon_AO'!(X:float=@n2, BasePose:Pose Link Structure=@n9, Alpha:float=1) -> (Pose:Pose Link Structure) {
    blendSpace=/Game/Animation/Game/MainPlayer/Poses/Medium/Male/AO/ExplosiveBow/Weapon/ExplosiveBow_ChargingLoop_Weapon_AO
    alphaBoolBlend {blendOption=Linear}
   }
   n11 = Two Way Blend!(A=Component To Local(ComponentPose=@n8), B=@n10, bAlphaBoolEnabled:bool=Get bIsBowAtMaxCharge() -> (bIsBowAtMaxCharge:bool)) -> (Pose) {blendNode {alphaInputType=Bool, bResetChildOnActivation=true, alphaBoolBlend {blendInTime=0.5, blendOption=Linear}}}
   Output Animation Pose(Result=Apply Additive!(
     Base=Apply Additive!(Base=@n11, Additive=@n1, Alpha:float) -> (Pose) {alphaBoolBlend {blendOption=Linear}}
     Additive="Sequence Player (additive) 'ExplosiveBow_Crouch_Charging_Into_CMM_Pose_Weapon'"() {sequence=/Game/Animation/Game/MainPlayer/Combat/Guns/ExplosiveBow/ExplosiveBow_Crouch_Charging_Into_CMM_Pose_Weapon}
     bAlphaBoolEnabled:bool=Get bIsCrouchMoving() -> (bIsCrouchMoving:bool)
    ) -> (Pose) {alphaInputType=Bool, alphaBoolBlend {blendInTime=0.45, blendOutTime=0.45}})
  ]
 }
 Transition {
  path=AnimGraph.AnimGraphNode_StateMachine_0.ExplosiveBow.AnimStateTransitionNode_0.Transition
  nodes [Result(bCanEnterTransition=Get bEnterChargingNormalTransition() -> (bEnterChargingNormalTransition:bool))]
 }
 Transition {
  path=AnimGraph.AnimGraphNode_StateMachine_0.ExplosiveBow.AnimStateTransitionNode_1.Transition
  nodes [Result(bCanEnterTransition=Get bEnterChargingFastTransition() -> (bEnterChargingFastTransition:bool))]
 }
 Transition {
  path=AnimGraph.AnimGraphNode_StateMachine_0.ExplosiveBow.AnimStateTransitionNode_4.Transition
  nodes [Result(bCanEnterTransition=NOT Boolean(A=Get bIsBowCharging() -> (bIsBowCharging:bool)))]
 }
 GrapplerBowCharge {
  entry=NotCharged
  path=AnimGraph.AnimGraphNode_StateMachine_1.GrapplerBowCharge
  ids=id
  nodes [
   AnimStateNode_0 = NotCharged() {is=title AnimStateNode}
   AnimStateNode_1 = ToCharged() {is=title AnimStateNode}
   AnimStateNode_2 = Charged() {is=title AnimStateNode}
   AnimStateTransitionNode_0 = AnimStateTransitionNode() {crossfadeDuration=0}
   AnimStateTransitionNode_1 = AnimStateTransitionNode() {bAutomaticRuleBasedOnSequencePlayerInState=true, crossfadeDuration=0}
   AnimStateTransitionNode_2 = AnimStateTransitionNode()
  ]
  links [
   {
    dst=NotCharged
    rule=AnimGraph.AnimGraphNode_StateMachine_1.GrapplerBowCharge.AnimStateTransitionNode_2.Transition
    src=Charged
    via=AnimStateTransitionNode_2
   }
   {
    dst=ToCharged
    rule=AnimGraph.AnimGraphNode_StateMachine_1.GrapplerBowCharge.AnimStateTransitionNode_0.Transition
    src=NotCharged
    via=AnimStateTransitionNode_0
   }
   {
    dst=Charged
    rule=AnimGraph.AnimGraphNode_StateMachine_1.GrapplerBowCharge.AnimStateTransitionNode_1.Transition
    src=ToCharged
    via=AnimStateTransitionNode_1
   }
  ]
 }
 NotCharged {
  path=AnimGraph.AnimGraphNode_StateMachine_1.GrapplerBowCharge.AnimStateNode_0.NotCharged
  nodes [Output Animation Pose(Result=Sequence Player 'GrapplerBow_NotCharged_GB'() {sequence=&/GrapplerBow_NotCharged_GB})]
 }
 ToCharged {
  path=AnimGraph.AnimGraphNode_StateMachine_1.GrapplerBowCharge.AnimStateNode_1.ToCharged
  nodes [Output Animation Pose(Result=Sequence Player 'GrapplerBow_NotCharged_To_Charged_GB'() {bLoopAnimation=false, sequence=&/GrapplerBow_NotCharged_To_Charged_GB})]
 }
 Charged {
  path=AnimGraph.AnimGraphNode_StateMachine_1.GrapplerBowCharge.AnimStateNode_2.Charged
  nodes [Output Animation Pose(Result=Sequence Player 'GrapplerBow_Charged_GB'() {sequence=&/GrapplerBow_Charged_GB})]
 }
 Transition {
  path=AnimGraph.AnimGraphNode_StateMachine_1.GrapplerBowCharge.AnimStateTransitionNode_0.Transition
  nodes [Result(bCanEnterTransition=Get bIsBowAtMaxCharge())]
 }
 Transition {path=AnimGraph.AnimGraphNode_StateMachine_1.GrapplerBowCharge.AnimStateTransitionNode_1.Transition, nodes [Result()]}
 Transition {
  path=AnimGraph.AnimGraphNode_StateMachine_1.GrapplerBowCharge.AnimStateTransitionNode_2.Transition
  nodes [Result(bCanEnterTransition=NOT Boolean(A=Get bIsBowAtMaxCharge()))]
 }
 EventGraph {nodes [
   Event BlueprintUpdateAnimation()
   n1 = GetOwningActor()
   > n2 = Is Valid(InputObject=@n1)
   Set GrapplerBow(execute=@n2.Is Not Valid, GrapplerBow:double=0) -> (Output_Get:double)
   n3 = Cast To B_Weap_Bow_Grappler_Athena!(execute=@n2.Is Valid, Object=@n1) -> (then, CastFailed, AsB Weap Bow Grappler Athena:B Weap Bow Grappler Athena Object Reference) {is=title K2Node_DynamicCast}
   > Set GrapplerBow(GrapplerBow=1)
   Set GrapplerBow(execute=@n3.CastFailed, GrapplerBow=0)
  ]}
]
variables [GrapplerBow {type=number}]
